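; Wind trail actor for the TKThrow ability that adds trail effects and sound to a thrown physics object.
; Setup stores the object and trail size and plays a throw sound; the size's RepNotify then activates the trail.
; Activation attaches to the object, starts Niagara and in-air audio, then self-destructs after a set duration.
; Object hits fire a local GameplayCue burst under a brief cooldown and fade out the audio.
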(event Collision|EventActorBeginOverlap
  (params (OtherActor :type Actor%20Object%20Reference)))

(event EventTick
  (params (DeltaSeconds :type Float%20%28single-precision%29)))

(event Custom|SetupWindTrail
  (params (PhysicObject :type Actor%20Object%20Reference) (Size :type Integer))
  (Variables|Default|SetPhysicObjectToApply PhysicObject)
  (Variables|Default|SetWindTrailSize Size)
  (CallFunction|PlayLibrarySound "(TagName=\"SoundLibrary.Ability.Object.Throw\")" (Variables|/Script/Engine.Actor|GetRootComponent (Variables|Default|GetPhysicObjectToApply))))

(event Custom|ActivateWindTrail
  (bind _windtrail (Variables|Default|GetWindTrail))
  (bind _physicobjecttoapply (Variables|Default|GetPhysicObjectToApply))
  (Collision|BindEventtoOnActorHit _physicobjecttoapply (delegate Custom|ActorHitSomething))
  (Transformation|AttachActorToActor self _physicobjecttoapply "None" "KeepWorld" "KeepWorld" "KeepWorld" true)
  (Niagara|SetNiagaraVariable(Int32) _windtrail "FXWindTrailSize" (Variables|Default|GetWindTrailSize))
  (Components|Activation|Activate _windtrail false)
  (bind (_success _output) (CallFunction|PlayLibrarySound "(TagName=\"SoundLibrary.Ability.Object.WhileInAir\")" (Variables|Default|GetWindTrail)))
  (bind _output_get (Variables|Default|SetInAirAudio _output))
  (Audio|Components|Audio|FadeIn _output_get 0.5 1.0 0.0 "Linear" :!class /Script/Engine.AudioComponent)
  (Utilities|FlowControl|Delay (ScalableFloat|GetValueatLevel (Variables|Default|GetWindTrailDuration)))
  (Collision|UnbindEventfromOnActorHit (Variables|Default|GetPhysicObjectToApply) (delegate Custom|ActorHitSomething))
  (Actor|DestroyActor))

(event Custom|ActorHitSomething
  (params (SelfActor :type Actor%20Object%20Reference) (OtherActor :type Actor%20Object%20Reference) (NormalImpulse :type Vector) (Hit :type Hit%20Result%20Structure%20%28by%20ref%29))
  (bind (_bblockinghit _binitialoverlap _time _distance _location _impactpoint) (Collision|BreakHitResult Hit))
  (if (not (Variables|Default|GetbHitGC_OnCooldown))
    (GameplayCue|ExecuteGameplayCue%28Burst%29Local (Variables|Default|GetPhysicObjectToApply) (Variables|Default|GetGC_ObjHit) (Ability|GameplayCue|MakeGameplayCueParameters :NormalizedMagnitude 0.0 :RawMagnitude 0.0 :Location _impactpoint :Normal "0,0,0" :Instigator (Game|GetInstigator) :EffectCauser OtherActor :GameplayEffectLevel 1 :AbilityLevel 1 :bReplicateLocationWhenUsingMinimalRepProxy false))
    (bind _inairaudio (Variables|Default|GetInAirAudio)
      (:then
        (Audio|Components|Audio|FadeOut _inairaudio 1.0 0.0 "Linear" :!class /Script/Engine.AudioComponent))
      (:else
        _))
    (Variables|Default|SetbHitGC_OnCooldown true)
    (Utilities|FlowControl|Delay 0.1)
    (Variables|Default|SetbHitGC_OnCooldown false)))

(fn OnRep_PhysicObjectToApply)

(fn OnRep_WindTrailSize
  (CallFunction|ActivateWindTrail))

(fn PlayLibrarySound
  (params (Event%20Name :type Gameplay%20Tag%20Structure) (Attach%20to%20Component :type Scene%20Component%20Object%20Reference))
  (bind _returnvalue (Game|GetInstigator))
  (bind _returnvalue_2 (WorldSubsystems|GetSoundLibrarySubsystem))
  ;[1 1 1] Get Sound library from instigator pawn. The play sound based on hit tag from pawn library
  (Utilities|IsValid _returnvalue_2
    (:Is%20Valid
      (Utilities|IsValid _returnvalue
        (:Is%20Valid
          (bind _returnvalue_3 (SoundLibrary|PlaySoundSimple _returnvalue_2 _returnvalue (Utilities|Struct|MakeSimpleSoundSettings Event%20Name true true true true 1.0 1.0 0.0 0 0 "None" Attach%20to%20Component) :!class /Script/SoundLibrary.SoundLibrarySubsystem))
          (return (.bSuccess (Utilities|Struct|BreakSoundLibraryPlayResult _returnvalue_3)) (Utilities|Array|Get(acopy) (.AudioComponents (Utilities|Struct|BreakSoundLibraryPlayResult _returnvalue_3)))))))))

(fn ConstructionScript)
