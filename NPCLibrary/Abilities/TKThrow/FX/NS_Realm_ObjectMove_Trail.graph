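# Niagara ribbon trail for the TKThrow ability, apparently following moved objects.
# Two emitters spawn by rate and distance: one with a lerped position offset and two ribbon renderers, the other shorter-lived with one renderer and colour-over-life scaling.
# Both taper width by curves scaled from User.FXWindTrailSize, run custom material UV modules and fade alpha.
# Alternate width and stretched-position attributes are also written; other user parameters have no visible use.
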NiagaraSystem /NPCLibrary/Abilities/TKThrow/FX/NS_Realm_ObjectMove_Trail
dependencies {
 usedDataInterfaces [&NiagaraDataInterfaceCurve, &NiagaraDataInterfaceColorCurve]
 usedDynamicInputs []
 usedModules [&SystemState]
}
emitters {
 Empty {
  emitterUpdateScript [
   &EmitterState(Life Cycle Mode=::NewEnumerator1, Loop Duration=2, UseLoopDelay=false)
   &SpawnRate(SpawnRate=10, Use Spawn Probability=false)
   &SpawnPerUnit(
    Module.Use Max Movement Threshold=true
    Module.Use Movement Tolerance=true
    Movement Tolerance=50
    Spawn Probability=1
    Spawn Spacing=30
    Use Spawn Probability=true
   )
  ]
  particleSpawnScript [
   &InitializeParticle(Color Mode=::NewEnumerator0, Lifetime=5, UsePositionOffset=false, Write Material Random=true)
   &InitializeRibbon(
    Lifetime=0.5
    Position Offset:Vector3f=/Niagara/DynamicInputs/Lerp/Lerp_Position(A:NiagaraPosition=<0 0 50>, B:NiagaraPosition=<0 0 1>, Alpha:NiagaraFloat=$Emitter.LoopedAge)
    Ribbon Width=50
    UsePositionOffset=true
    Write Material Random=true
   )
  ]
  particleUpdateScript [
   &ParticleState()
   /Game/Effects/Niagara/Modules/Materials/MaterialUV()
   &Material()
   /SetVariables(Particles.RibbonWidth=&FloatFromCurve(
     FloatCurve=DataInterface() {propertyValues {Curve {keys [
         {arriveTangent=3.6621, interpMode=RCIM_Cubic, leaveTangent=3.6621, tangentMode=RCTM_User, time=0, value=1.2}
         {arriveTangent=-1.674, interpMode=RCIM_Cubic, leaveTangent=-1.674, tangentMode=RCTM_User, time=0.5, value=0.9}
         {arriveTangent=-0.0405, interpMode=RCIM_Cubic, leaveTangent=-0.0405, tangentMode=RCTM_User, time=1, value=0.3}
        ]}}}
     Scale Curve=&Multiply_Float(A=&MakeFloatFromInt(Int=$User.FXWindTrailSize), B=40)
    )) {isSetParameters=true}
   /SetVariables(Particles.RibbonWidthAlt1:NiagaraFloat=&Multiply_Float(A=$Particles.RibbonWidth, B=1)) {isSetParameters=true, module=SetVariables_#1}
   /SetVariables(Particles.PositionAlt1:NiagaraPosition=&Multiply_Vector(A=Linked() {linkedVariable {name=Particles.Position, type=NiagaraPosition}}, B=<1 1 1.5>)) {isSetParameters=true}
   &Color()
   &ScaleColor(
    Scale Alpha=&FloatFromCurve(FloatCurve=DataInterface() {propertyValues {Curve {keys [
         <at 0.2 0 RCIM_Cubic>
         {arriveTangent=0.0115, interpMode=RCIM_Cubic, leaveTangent=0.0115, tangentMode=RCTM_User, time=0.3, value=0.25}
         {arriveTangent=-0.0924, interpMode=RCIM_Cubic, leaveTangent=-0.0924, tangentMode=RCTM_User, time=0.4, value=0.1}
         <at 1.0011 0.002 RCIM_Cubic>
        ]}}})
    Scale RGB=&MakeVectorFromLinearColorRGB(Color=&ColorFromCurve(ColorCurve=DataInterface() {propertyValues {
        AlphaCurve {keys [<at 0 1 RCIM_Cubic>,<at 1 0 RCIM_Cubic>]}
        BlueCurve {keys [<at 0.1103 1 RCIM_Cubic>,<at 0.4987 1.177>,<at 1 1 RCIM_Cubic>]}
        GreenCurve {keys [<at 0.1103 1 RCIM_Cubic>,<at 0.4987 1.25>,<at 1 0.4431 RCIM_Cubic>]}
        RedCurve {keys [<at 0.1103 1 RCIM_Cubic>,<at 0.4987 0.7682>,<at 1 0.7914 RCIM_Cubic>]}
       }}))
    ScaleA=true
    ScaleRGB=false
   ) {module=ScaleColor001}
  ]
  rendererClasses [&NiagaraRibbonRendererProperties]
  renderers [
   {rendererClass=&NiagaraRibbonRendererProperties}
   {rendererClass=&NiagaraRibbonRendererProperties, rendererIndex=1}
  ]
 }
 Empty001 {
  emitterUpdateScript [
   &EmitterState()
   &SpawnRate(SpawnRate=30, Use Spawn Probability=false)
   /Game/Effects/Niagara/Modules/Spawn/FN_SpawnByEngineDistance(Step=15)
  ]
  particleSpawnScript [
   &InitializeParticle(Color Mode=::NewEnumerator0, Lifetime=5, UsePositionOffset=false, Write Material Random=true)
   &InitializeRibbon(Lifetime=0.4, Ribbon Width=50, UsePositionOffset=false, Write Material Random=true)
  ]
  particleUpdateScript [
   &ParticleState()
   &MaterialUV()
   &Material()
   /SetVariables(Particles.RibbonWidth=&FloatFromCurve(
     FloatCurve=DataInterface() {propertyValues {Curve {keys [
         {arriveTangent=-0.407, interpMode=RCIM_Cubic, leaveTangent=-0.407, tangentMode=RCTM_User, time=0, value=1.2}
         {arriveTangent=-4.4401, interpMode=RCIM_Cubic, leaveTangent=-4.4401, tangentMode=RCTM_User, time=0.6093, value=0.778}
         {arriveTangent=-0.0405, interpMode=RCIM_Cubic, leaveTangent=-0.0405, tangentMode=RCTM_User, time=1, value=0.3}
        ]}}}
     Scale Curve=&Multiply_Float(A=&MakeFloatFromInt(Int=$User.FXWindTrailSize), B=20)
    )) {isSetParameters=true}
   /SetVariables(Particles.RibbonWidthAlt1:NiagaraFloat=&Multiply_Float(A=$Particles.RibbonWidth, B=2)) {isSetParameters=true, module=SetVariables_#1}
   &Color()
   &ScaleColor(
    Scale Alpha=&FloatFromCurve(
     FloatCurve=DataInterface() {propertyValues {Curve {keys [
         <at 0 0 RCIM_Cubic>
         {arriveTangent=0.0115, interpMode=RCIM_Cubic, leaveTangent=0.0115, tangentMode=RCTM_User, time=0.1, value=0.25}
         {arriveTangent=-0.0924, interpMode=RCIM_Cubic, leaveTangent=-0.0924, tangentMode=RCTM_User, time=0.2615, value=0.0989}
         <at 1.0011 0.002 RCIM_Cubic>
        ]}}}
     Scale Curve=4
    )
    Scale RGB=&MakeVectorFromLinearColorRGB(Color=&ColorFromCurve(ColorCurve=DataInterface() {propertyValues {
        AlphaCurve {keys [<at 0 1 RCIM_Cubic>,<at 1 0 RCIM_Cubic>]}
        BlueCurve {keys [<at 0 0 RCIM_Cubic>,<at 0.4987 1.177>,<at 1 1 RCIM_Cubic>]}
        GreenCurve {keys [<at 0 0 RCIM_Cubic>,<at 0.4987 1.25>,<at 1 0.4431 RCIM_Cubic>]}
        RedCurve {keys [<at 0 0 RCIM_Cubic>,<at 0.4987 0.7682>,<at 1 0.7914 RCIM_Cubic>]}
       }}))
    ScaleA=true
    ScaleRGB=true
   ) {module=ScaleColor001}
  ]
  rendererClasses [&NiagaraRibbonRendererProperties]
  renderers [{rendererClass=&NiagaraRibbonRendererProperties}]
 }
}
summary {
 systemName=NS_Realm_ObjectMove_Trail
 emitters [
  {emitterName=Empty001, rendererClasses [&NiagaraRibbonRendererProperties]}
  {emitterName=Empty, rendererClasses [&NiagaraRibbonRendererProperties]}
 ]
 userVariables [
  User.Color2Green {type=LinearColor, description="", defaultValue {struct=LinearColor, value=<rgba 1 1 1 1>}}
  User.FXWindTrailSize {type=NiagaraInt32, defaultValue=2}
  User.RibbonIDOffset {type=NiagaraInt32}
  User.SkeletalMesh {
   type=NiagaraDataInterfaceSkeletalMesh
   defaultValue {
    struct=NiagaraExt_VariableValue_DataInterface
    value {dataInterface=&self:NiagaraDataInterfaceSkeletalMesh_0}
   }
  }
 ]
}
